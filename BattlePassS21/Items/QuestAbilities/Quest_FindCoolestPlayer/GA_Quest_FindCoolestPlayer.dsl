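; Gameplay ability for the Battle Pass S21 "Find Coolest Player" quest that places a quest actor at nearby public mirrors.
; On event activation, it sphere-traces around the owning pawn and spawns the quest actor at each Apollo shared mirror hit, storing only the latest spawn.
; It also waits for the Athena.Quests.MirrorTagAdd tag to leave the pawn, then destroys the stored actor if valid and ends the ability.
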(event Ability|EventActivateAbilityFromEvent
  (params (EventData :type Gameplay%20Event%20Data%20Structure%20%28by%20ref%29))
  (bind _spawned_actor (Variables|Default|GetSpawned%20Actor))
  (|CallParentFunction EventData)
  (Ability|Tasks|WaitGameplayTagRemoveFromActor (Variables|GAVars|Data|GetOwningPlayerPawn) "(TagName=\"Athena.Quests.MirrorTagAdd\")"
    (:then
      (bind _outhits (Collision|MultiSphereTraceForObjects (Transformation|GetActorLocation (Variables|GAVars|Data|GetOwningPlayerPawn)) (Transformation|GetActorLocation (Variables|GAVars|Data|GetOwningPlayerPawn)) 3e+02 (Variables|Default|GetObject%20Types)))
      (for _array_element _outhits
        (if (.bSuccess (Utilities|Casting|CastToApollo_Shared_PublicMirror_01 (.HitActor (Collision|BreakHitResult _array_element))))
          (bind (_asynctaskproxy _spawnedactor) (GameplayTasks|SpawnActorforGameplayTask 0 (Transformation|GetActorLocation (Utilities|Casting|CastToApollo_Shared_PublicMirror_01 (.HitActor (Collision|BreakHitResult _array_element)))) "0, 0, 0" "/BattlePassS21/Items/QuestAbilities/Quest_FindCoolestPlayer/BP_S21_GA_QuestActor_FindCoolestPlayer.BP_S21_GA_QuestActor_FindCoolestPlayer_C" false 0 0 0)
            (:Success
              (Variables|Default|SetSpawned%20Actor _spawnedactor))))))
    (:Removed
      (if (Utilities|IsValid _spawned_actor :!class /Script/Engine.KismetSystemLibrary)
        (Actor|DestroyActor _spawned_actor)
        (else
          _))
      (Ability|EndAbility))))
